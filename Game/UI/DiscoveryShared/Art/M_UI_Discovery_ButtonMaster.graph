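# Translucent UI master material for Discovery buttons, drawing an animatable fill, border and outer alpha mask.
# Border and alpha-mask shapes come from rotated UV edge distances sharpened and saturated, with optional per-side asymmetry and box-scale UV selection.
# Hover_Animate blends edge shapes and colors, while Pressed_Animate and bIsDisabled blend further color states.
# Static switches toggle border, color and shape animation, alpha channels and outer masking, with AlphaOveride adjusting final opacity.
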Material /Game/UI/DiscoveryShared/Art/M_UI_Discovery_ButtonMaster
graph {
 properties {
  bUsedWithStaticMesh=false
  blendMode=BLEND_Translucent
  materialDomain=MD_UI
  pixelDepthOffsetMode=PDOM_Legacy
  usedShadingModels=Default Lit
 }
 nodes [
  n1 = VectorParameter() {group=Colors, parameterName=Base_Fill, defaultValue=<rgba 0 0.0908 1 1>}
  n2 = VectorParameter() {group=Colors, parameterName=Base_Border, defaultValue=<rgba 0.0097 0.2623 1 1>}
  n3 = NamedRerouteUsage()
  n4 = LinearInterpolate(A=@n1.A, B=@n2.A, Alpha=@n3)
  n5 = VectorParameter() {group=Colors, parameterName=Hover_Fill, defaultValue=<rgba 0.0097 0.2623 1 1>}
  n6 = VectorParameter() {group=Colors, parameterName=Hover_Border, defaultValue=<rgba 0.2159 0.6584 1 1>}
  n7 = NamedRerouteUsage()
  n8 = LinearInterpolate(A=@n5.A, B=@n6.A, Alpha=@n7)
  n9 = ScalarParameter() {group=Animators, parameterName=Hover_Animate, sliderMax=1}
  n10 = VectorParameter() {group=Colors, parameterName=Pressed_Fill, defaultValue=<rgba 1 1 1 1>}
  n11 = VectorParameter() {group=Colors, parameterName=Pressed_Border, defaultValue=<rgba 0.6038 0.9647 1 1>}
  n12 = NamedRerouteUsage()
  n13 = ScalarParameter() {group=Animators, parameterName=Pressed_Animate, sliderMax=1}
  n14 = VectorParameter() {group=Colors, parameterName=Disabled_Fill, defaultValue=<rgba 0.0467 0.107 0.3185 1>}
  n15 = VectorParameter() {group=Colors, parameterName=Disabled_Border, defaultValue=<rgba 0.1329 0.3185 0.6038 1>}
  n16 = NamedRerouteUsage()
  n17 = ScalarParameter() {group=Switches, parameterName=bIsDisabled, sliderMax=1}
  n18 = LinearInterpolate(
   A=LinearInterpolate(A=LinearInterpolate(A=@n4, B=@n8, Alpha=@n9), B=LinearInterpolate(A=@n10.A, B=@n11.A, Alpha=@n12), Alpha=@n13)
   B=LinearInterpolate(A=@n14.A, B=@n15.A, Alpha=@n16)
   Alpha=@n17
  )
  n19 = VectorParameter() {group=Colors, parameterName=Base_Fill, defaultValue=<rgba 0 0.0908 1 1>}
  n20 = VectorParameter() {group=Colors, parameterName=Hover_Fill, defaultValue=<rgba 0.0097 0.2623 1 1>}
  n21 = ScalarParameter() {group=Animators, parameterName=Hover_Animate, sliderMax=1}
  n22 = VectorParameter() {group=Colors, parameterName=Pressed_Fill, defaultValue=<rgba 1 1 1 1>}
  n23 = ScalarParameter() {group=Animators, parameterName=Pressed_Animate, sliderMax=1}
  n24 = VectorParameter() {group=Colors, parameterName=Disabled_Fill, defaultValue=<rgba 0.0467 0.107 0.3185 1>}
  n25 = ScalarParameter() {group=Switches, parameterName=bIsDisabled, sliderMax=1}
  n26 = LinearInterpolate(A=LinearInterpolate(A=LinearInterpolate(A=@n19.A, B=@n20.A, Alpha=@n21), B=@n22.A, Alpha=@n23), B=@n24.A, Alpha=@n25)
  n27 = StaticSwitchParameter(
   True=StaticSwitchParameter(
    True=StaticSwitchParameter(True=@n18, False=@n4) {defaultValue=true, group=Switches, parameterName=AnimateColor}
    False=StaticSwitchParameter(True=@n26, False=@n19.A) {defaultValue=true, group=Switches, parameterName=AnimateColor}
   ) {defaultValue=true, group=Switches, parameterName=bHasBorder}
   False=Constant() {r=1}
  ) {group=Switches, parameterName=bUseAlphaChannels}
  n28 = NamedRerouteUsage()
  LinearInterpolate(
   B=ScalarParameter() {defaultValue=1, group=Animators, parameterName=AlphaOveride, sliderMax=1}
   Alpha=StaticSwitchParameter(True=LinearInterpolate(A=@n27, Alpha=@n28) {constB=0}, False=@n27) {group=Switches, parameterName=bUseOuterAlphaMask}
  )
  n29 = TextureCoordinate() {coordinateIndex=4}
  n30 = TextureCoordinate()
  n31 = ScalarParameter() {group=Switches, parameterName=UseBoxScaleV, sliderMax=1}
  n32 = Multiply(A=Add(A=LinearInterpolate(A=@n29, B=@n30, Alpha=@n31)) {constB=-0.5}) {constB=1024}
  NamedRerouteDeclaration(None=ComponentMask(None=Saturate(None=@n32)) {r=true}) {name=YMask, nodeColor=<rgba 0 1 0.4002 1>}
  n33 = ScalarParameter() {group=Switches, parameterName=UseBoxScaleU, sliderMax=1}
  n34 = Multiply(A=Add(A=LinearInterpolate(A=@n29, B=@n30, Alpha=@n33)) {constB=-0.5}) {constB=1024}
  NamedRerouteDeclaration(None=ComponentMask(None=Saturate(None=@n34)) {g=true}) {name=XMask, nodeColor=<rgba 0 0.4207 1 1>}
  n35 = ScalarParameter() {defaultValue=-0.0125, group=Border and Alpha Mask, parameterName=Border_Rotation_V_A}
  n36 = ScalarParameter() {group=Animators, parameterName=Hover_Animate, sliderMax=1}
  n37 = StaticSwitchParameter(
   True=LinearInterpolate(A=@n35, B=ScalarParameter() {defaultValue=-0.0063, group=Border and Alpha Mask, parameterName=Border_Rotation_V_B}, Alpha=@n36)
   False=@n35
  ) {defaultValue=true, group=Switches, parameterName=AnimateShape}
  n38 = TextureCoordinate() {coordinateIndex=4}
  n39 = TextureCoordinate()
  n40 = LinearInterpolate(A=@n38, B=@n39, Alpha=ScalarParameter() {group=Switches, parameterName=UseBoxScaleV, sliderMax=1})
  n41 = Add(A=Rotator(Coordinate=@n40, Time=@n37) {speed=6.2832}) {constB=-0.5}
  n42 = ScalarParameter() {defaultValue=0.125, group=Border and Alpha Mask, parameterName=Border_Edge_V_A}
  n43 = ScalarParameter() {defaultValue=0.25, group=Border and Alpha Mask, parameterName=Border_Edge_V_B}
  n44 = ScalarParameter() {group=Animators, parameterName=Hover_Animate, sliderMax=1}
  n45 = Add(A=StaticSwitchParameter(True=LinearInterpolate(A=@n42, B=@n43, Alpha=@n44), False=@n42) {defaultValue=true, group=Switches, parameterName=AnimateShape}) {constB=-0.5}
  n46 = ComponentMask(None=Add(A=Abs(None=@n41), B=@n45)) {r=true}
  n47 = ScalarParameter() {defaultValue=-0.0125, group=Border and Alpha Mask, parameterName=Border_Rotation_V_Right_A}
  n48 = ScalarParameter() {defaultValue=-0.0063, group=Border and Alpha Mask, parameterName=Border_Rotation_V_Right_B}
  n49 = StaticSwitchParameter(
   True=StaticSwitchParameter(True=LinearInterpolate(A=@n47, B=@n48, Alpha=@n36), False=@n47) {defaultValue=true, group=Switches, parameterName=AnimateShape}
   False=Multiply(A=@n37) {constB=-1}
  ) {group=Switches, parameterName=AsymmetricalVAxis}
  n50 = Abs(None=Add(A=Rotator(Coordinate=@n40, Time=@n49) {speed=6.2832}) {constB=-0.5})
  n51 = ScalarParameter() {defaultValue=0.125, group=Border and Alpha Mask, parameterName=Border_Edge_V_Right_A}
  n52 = LinearInterpolate(A=@n51, B=ScalarParameter() {defaultValue=0.25, group=Border and Alpha Mask, parameterName=Border_Edge_V_Right_B}, Alpha=@n44)
  n53 = StaticSwitchParameter(
   True=Add(A=StaticSwitchParameter(True=@n52, False=@n51) {defaultValue=true, group=Switches, parameterName=AnimateShape}) {constB=-0.5}
   False=@n45
  ) {group=Switches, parameterName=AsymmetricalVAxis}
  n54 = LinearInterpolate(A=@n46, B=ComponentMask(None=Add(A=@n50, B=@n53)) {r=true}, Alpha=NamedRerouteUsage())
  n55 = ScalarParameter() {defaultValue=64, group=Border and Alpha Mask, parameterName=Sharpness_V}
  n56 = StaticSwitchParameter(True=Saturate(None=Multiply(A=@n54, B=@n55)), False=Saturate(None=Multiply(A=@n46, B=@n55))) {group=Switches, parameterName=MirrorOrAsymmetricalVAxis}
  n57 = ScalarParameter() {defaultValue=0.0063, group=Border and Alpha Mask, parameterName=Border_Rotation_U_A}
  n58 = StaticSwitchParameter(
   True=LinearInterpolate(A=@n57, B=ScalarParameter() {defaultValue=0.0031, group=Border and Alpha Mask, parameterName=Border_Rotation_U_B}, Alpha=@n36)
   False=@n57
  ) {defaultValue=true, group=Switches, parameterName=AnimateShape}
  n59 = LinearInterpolate(A=@n38, B=@n39, Alpha=ScalarParameter() {group=Switches, parameterName=UseBoxScaleU, sliderMax=1})
  n60 = Add(A=Rotator(Coordinate=@n59, Time=@n58) {speed=6.2832}) {constB=-0.5}
  n61 = ScalarParameter() {defaultValue=0.125, group=Border and Alpha Mask, parameterName=Border_Edge_U_A}
  n62 = ScalarParameter() {defaultValue=0.25, group=Border and Alpha Mask, parameterName=Border_Edge_U_B}
  n63 = Add(A=StaticSwitchParameter(True=LinearInterpolate(A=@n61, B=@n62, Alpha=@n44), False=@n61) {defaultValue=true, group=Switches, parameterName=AnimateShape}) {constB=-0.5}
  n64 = ComponentMask(None=Add(A=Abs(None=@n60), B=@n63)) {g=true}
  n65 = ScalarParameter() {defaultValue=0.0063, group=Border and Alpha Mask, parameterName=Border_Rotation_U_Bottom_A}
  n66 = ScalarParameter() {defaultValue=0.0031, group=Border and Alpha Mask, parameterName=Border_Rotation_U_Bottom_B}
  n67 = StaticSwitchParameter(
   True=StaticSwitchParameter(True=LinearInterpolate(A=@n65, B=@n66, Alpha=@n36), False=@n65) {defaultValue=true, group=Switches, parameterName=AnimateShape}
   False=Multiply(A=@n58) {constB=-1}
  ) {group=Switches, parameterName=AsymmetricalUAxis}
  n68 = Abs(None=Add(A=Rotator(Coordinate=@n59, Time=@n67) {speed=6.2832}) {constB=-0.5})
  n69 = ScalarParameter() {defaultValue=0.125, group=Border and Alpha Mask, parameterName=Border_Edge_U_Bottom_A}
  n70 = LinearInterpolate(A=@n69, B=ScalarParameter() {defaultValue=0.25, group=Border and Alpha Mask, parameterName=Border_Edge_U_Bottom_B}, Alpha=@n44)
  n71 = StaticSwitchParameter(
   True=Add(A=StaticSwitchParameter(True=@n70, False=@n69) {defaultValue=true, group=Switches, parameterName=AnimateShape}) {constB=-0.5}
   False=@n63
  ) {group=Switches, parameterName=AsymmetricalUAxis}
  n72 = LinearInterpolate(A=@n64, B=ComponentMask(None=Add(A=@n68, B=@n71)) {g=true}, Alpha=NamedRerouteUsage())
  n73 = ScalarParameter() {defaultValue=64, group=Border and Alpha Mask, parameterName=Sharpness_U}
  n74 = StaticSwitchParameter(True=Saturate(None=Multiply(A=@n72, B=@n73)), False=Saturate(None=Multiply(A=@n64, B=@n73))) {group=Switches, parameterName=MirrorOrAsymmetricalUAxis}
  NamedRerouteDeclaration(None=Saturate(None=Add(A=@n56, B=@n74))) {name=Border, nodeColor=<rgba 0.8685 0 1 1>}
  n75 = ScalarParameter() {defaultValue=-0.0125, group=Border and Alpha Mask, parameterName=Mask_Rotation_V_A}
  n76 = ScalarParameter() {group=Animators, parameterName=Hover_Animate, sliderMax=1}
  n77 = StaticSwitchParameter(
   True=LinearInterpolate(A=@n75, B=ScalarParameter() {defaultValue=-0.0063, group=Border and Alpha Mask, parameterName=Mask_Rotation_V_B}, Alpha=@n76)
   False=@n75
  ) {defaultValue=true, group=Switches, parameterName=AnimateShape}
  n78 = TextureCoordinate() {coordinateIndex=4}
  n79 = TextureCoordinate()
  n80 = LinearInterpolate(A=@n78, B=@n79, Alpha=ScalarParameter() {group=Switches, parameterName=UseBoxScaleV, sliderMax=1})
  n81 = Add(A=Rotator(Coordinate=@n80, Time=@n77) {speed=6.2832}) {constB=-0.5}
  n82 = ScalarParameter() {defaultValue=0.0625, group=Border and Alpha Mask, parameterName=Mask_Edge_V_A}
  n83 = ScalarParameter() {defaultValue=0.125, group=Border and Alpha Mask, parameterName=Mask_Edge_V_B}
  n84 = ScalarParameter() {group=Animators, parameterName=Hover_Animate, sliderMax=1}
  n85 = Add(A=StaticSwitchParameter(True=LinearInterpolate(A=@n82, B=@n83, Alpha=@n84), False=@n82) {defaultValue=true, group=Switches, parameterName=AnimateShape}) {constB=-0.5}
  n86 = ComponentMask(None=Add(A=Abs(None=@n81), B=@n85)) {r=true}
  n87 = ScalarParameter() {defaultValue=-0.0125, group=Border and Alpha Mask, parameterName=Mask_Rotation_V_Right_A}
  n88 = ScalarParameter() {defaultValue=-0.0063, group=Border and Alpha Mask, parameterName=Mask_Rotation_V_Right_B}
  n89 = StaticSwitchParameter(
   True=StaticSwitchParameter(True=LinearInterpolate(A=@n87, B=@n88, Alpha=@n76), False=@n87) {defaultValue=true, group=Switches, parameterName=AnimateShape}
   False=Multiply(A=@n77) {constB=-1}
  ) {group=Switches, parameterName=AsymmetricalVAxis}
  n90 = Abs(None=Add(A=Rotator(Coordinate=@n80, Time=@n89) {speed=6.2832}) {constB=-0.5})
  n91 = ScalarParameter() {defaultValue=0.0625, group=Border and Alpha Mask, parameterName=Mask_Edge_V_Right_A}
  n92 = LinearInterpolate(A=@n91, B=ScalarParameter() {defaultValue=0.125, group=Border and Alpha Mask, parameterName=Mask_Edge_V_Right_B}, Alpha=@n84)
  n93 = StaticSwitchParameter(
   True=Add(A=StaticSwitchParameter(True=@n92, False=@n91) {defaultValue=true, group=Switches, parameterName=AnimateShape}) {constB=-0.5}
   False=@n85
  ) {group=Switches, parameterName=AsymmetricalVAxis}
  n94 = LinearInterpolate(A=@n86, B=ComponentMask(None=Add(A=@n90, B=@n93)) {r=true}, Alpha=NamedRerouteUsage())
  n95 = ScalarParameter() {defaultValue=64, group=Border and Alpha Mask, parameterName=Sharpness_V}
  n96 = StaticSwitchParameter(True=Saturate(None=Multiply(A=@n94, B=@n95)), False=Saturate(None=Multiply(A=@n86, B=@n95))) {group=Switches, parameterName=MirrorOrAsymmetricalVAxis}
  n97 = ScalarParameter() {defaultValue=0.0063, group=Border and Alpha Mask, parameterName=Mask_Rotation_U_A}
  n98 = StaticSwitchParameter(
   True=LinearInterpolate(A=@n97, B=ScalarParameter() {defaultValue=0.0031, group=Border and Alpha Mask, parameterName=Mask_Rotation_U_B}, Alpha=@n76)
   False=@n97
  ) {defaultValue=true, group=Switches, parameterName=AnimateShape}
  n99 = LinearInterpolate(A=@n78, B=@n79, Alpha=ScalarParameter() {group=Switches, parameterName=UseBoxScaleU, sliderMax=1})
  n100 = Add(A=Rotator(Coordinate=@n99, Time=@n98) {speed=6.2832}) {constB=-0.5}
  n101 = ScalarParameter() {defaultValue=0.0625, group=Border and Alpha Mask, parameterName=Mask_Edge_U_A}
  n102 = ScalarParameter() {defaultValue=0.125, group=Border and Alpha Mask, parameterName=Mask_Edge_U_B}
  n103 = Add(A=StaticSwitchParameter(True=LinearInterpolate(A=@n101, B=@n102, Alpha=@n84), False=@n101) {defaultValue=true, group=Switches, parameterName=AnimateShape}) {constB=-0.5}
  n104 = ComponentMask(None=Add(A=Abs(None=@n100), B=@n103)) {g=true}
  n105 = ScalarParameter() {defaultValue=0.0063, group=Border and Alpha Mask, parameterName=Mask_Rotation_U_Bottom_A}
  n106 = ScalarParameter() {defaultValue=0.0031, group=Border and Alpha Mask, parameterName=Mask_Rotation_U_Bottom_B}
  n107 = StaticSwitchParameter(
   True=StaticSwitchParameter(True=LinearInterpolate(A=@n105, B=@n106, Alpha=@n76), False=@n105) {defaultValue=true, group=Switches, parameterName=AnimateShape}
   False=Multiply(A=@n98) {constB=-1}
  ) {group=Switches, parameterName=AsymmetricalUAxis}
  n108 = Abs(None=Add(A=Rotator(Coordinate=@n99, Time=@n107) {speed=6.2832}) {constB=-0.5})
  n109 = ScalarParameter() {defaultValue=0.0625, group=Border and Alpha Mask, parameterName=Mask_Edge_U_Bottom_A}
  n110 = LinearInterpolate(A=@n109, B=ScalarParameter() {defaultValue=0.125, group=Border and Alpha Mask, parameterName=Mask_Edge_U_Bottom_B}, Alpha=@n84)
  n111 = StaticSwitchParameter(
   True=Add(A=StaticSwitchParameter(True=@n110, False=@n109) {defaultValue=true, group=Switches, parameterName=AnimateShape}) {constB=-0.5}
   False=@n103
  ) {group=Switches, parameterName=AsymmetricalUAxis}
  n112 = LinearInterpolate(A=@n104, B=ComponentMask(None=Add(A=@n108, B=@n111)) {g=true}, Alpha=NamedRerouteUsage())
  n113 = ScalarParameter() {defaultValue=64, group=Border and Alpha Mask, parameterName=Sharpness_U}
  n114 = StaticSwitchParameter(True=Saturate(None=Multiply(A=@n112, B=@n113)), False=Saturate(None=Multiply(A=@n104, B=@n113))) {group=Switches, parameterName=MirrorOrAsymmetricalUAxis}
  NamedRerouteDeclaration(None=Saturate(None=Add(A=@n96, B=@n114))) {name=AlphaMask, nodeColor=<rgba 1 0 0.0762 1>}
  n115 = LinearInterpolate(A=@n1.RGB, B=@n2.RGB, Alpha=@n3)
  n116 = LinearInterpolate(
   A=LinearInterpolate(A=@n115, B=LinearInterpolate(A=@n5.RGB, B=@n6.RGB, Alpha=@n7), Alpha=@n9)
   B=LinearInterpolate(A=@n10.RGB, B=@n11.RGB, Alpha=@n12)
   Alpha=@n13
  )
  n117 = LinearInterpolate(A=@n14.RGB, B=@n15.RGB, Alpha=@n16)
  n118 = LinearInterpolate(A=LinearInterpolate(A=@n19.RGB, B=@n20.RGB, Alpha=@n21), B=@n22.RGB, Alpha=@n23)
  StaticSwitchParameter(
   True=StaticSwitchParameter(True=LinearInterpolate(A=@n116, B=@n117, Alpha=@n17), False=@n115) {defaultValue=true, group=Switches, parameterName=AnimateColor}
   False=StaticSwitchParameter(True=LinearInterpolate(A=@n118, B=@n24.RGB, Alpha=@n25), False=@n19.RGB) {defaultValue=true, group=Switches, parameterName=AnimateColor}
  ) {defaultValue=true, group=Switches, parameterName=bHasBorder}
 ]
}
