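# Translucent UI material that draws a colored outline border, by its name for Discovery tiles.
# A Rounded Corners switch picks either the CoreUI rounded-mask function or an anti-aliased distance-field mask built from centered UVs, with parameters for width.
# An optional Edge Mask, built the same distance-field way, is subtracted from a mask before scaling by Outline Opacity.
# Named reroutes share UVs, color, masks and anti-aliasing threshold between sections.
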Material /Game/UI/DiscoveryShared/Art/M_UI_Discovery_Tile_Border
graph {
 properties {
  bUsedWithStaticMesh=false
  blendMode=BLEND_Translucent
  materialDomain=MD_UI
  pixelDepthOffsetMode=PDOM_Legacy
  usedShadingModels=Default Lit
 }
 nodes [
  n1 = Subtract(A=NamedRerouteUsage(), B=NamedRerouteUsage())
  Multiply(
   A=StaticSwitchParameter(True=Saturate(None=@n1), False=NamedRerouteUsage()) {group=2 - Edge Mask, parameterName=Display Edge Mask, sortPriority=31}
   B=ScalarParameter() {defaultValue=1, group=1 - Outline, parameterName=Outline Opacity, sliderMax=1, sortPriority=1}
  )
  n2 = NamedRerouteUsage()
  n3 = Multiply(A=Abs(None=Subtract(A=@n2) {constB=0.5})) {constB=2}
  n4 = NamedRerouteUsage()
  n5 = OneMinus(None=Subtract(A=ScalarParameter() {defaultValue=0.02, group=2 - Edge Mask, parameterName=Outline Mask Size, sliderMax=1}, B=@n4))
  n6 = MaterialFunctionCall(Float2=Saturate(None=Subtract(A=@n3, B=@n5))) {materialFunction=&BreakOutFloat2Components}
  n7 = Saturate(None=Add(A=@n6.R, B=@n6.R))
  NamedRerouteDeclaration(None=Saturate(None=MaterialFunctionCall(DistanceField=@n7, Threshold=@n4) {materialFunction=&MF_UI_AntiAliasing}.Outer)) {name=Edge Mask, nodeColor=<rgba 0 0 0 1>}
  n8 = MaterialFunctionCall() {materialFunction=&GetUserInterfaceUV}
  NamedRerouteDeclaration(None=@n8.9-Slice UV) {name=UV, nodeColor=<rgba 0 0 0 1>}
  NamedRerouteDeclaration(None=VectorParameter() {group=1 - Outline, parameterName=Outline Color, sortPriority=0, defaultValue=<rgba 0 0.1163 1 1>}.RGB) {name=Outline Color, nodeColor=<rgba 0 0 0 1>}
  n9 = NamedRerouteUsage()
  n10 = ScalarParameter() {defaultValue=1, group=1 - Outline, parameterName=Outline Size}
  n11 = ScalarParameter() {defaultValue=0.1, group=1 - Outline, parameterName=Corner Radius}
  n12 = ScalarParameter() {defaultValue=0.05, group=1 - Outline, parameterName=Outline Thickness}
  n13 = NamedRerouteUsage()
  n14 = Multiply(A=Abs(None=Subtract(A=@n13) {constB=0.5})) {constB=2}
  n15 = OneMinus(None=NamedRerouteUsage())
  n16 = MaterialFunctionCall(Float2=Saturate(None=Subtract(A=@n14, B=@n15))) {materialFunction=&BreakOutFloat2Components}
  n17 = MaterialFunctionCall(
   DistanceField=Saturate(None=Add(A=@n16.R, B=@n16.R))
   Threshold=NamedRerouteDeclaration(None=NamedRerouteUsage()) {name=AA Threshold, nodeColor=<rgba 0 0 0 1>}
  ) {materialFunction=&MF_UI_AntiAliasing}
  NamedRerouteDeclaration(None=StaticSwitchParameter(
    True=MaterialFunctionCall(UV=@n9, Size=@n10, Corner Radius=@n11, Outline Thickness=@n12) {materialFunction=&MF_UI_CoreUI_RoundedMask}.Outline Mask
    False=Saturate(None=@n17.Outer)
   ) {group=0 - Features, parameterName=Rounded Corners}) {name=Outline Mask, nodeColor=<rgba 0 0 0 1>}
  NamedRerouteDeclaration(None=ScalarParameter() {defaultValue=0.01, group=1 - Outline, parameterName=Aliasing Threshold}) {name=AA Threshold, nodeColor=<rgba 0 0 0 1>}
  NamedRerouteDeclaration(None=@n8.Normalized UV) {name=Normalized UV, nodeColor=<rgba 0 0 0 1>}
  n18 = ScalarParameter() {defaultValue=0.1, group=1 - Outline, parameterName=Max Outline Size, sortPriority=2}
  n19 = ScalarParameter() {defaultValue=1, group=1 - Outline, parameterName=Outline Size Ratio, sliderMax=1, sortPriority=3}
  NamedRerouteDeclaration(None=Subtract(A=Multiply(A=@n18, B=@n19), B=NamedRerouteUsage())) {name=Outline Width, nodeColor=<rgba 0 0 0 1>}
  NamedRerouteUsage()
 ]
}
