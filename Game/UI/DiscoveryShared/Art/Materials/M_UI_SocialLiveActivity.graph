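# Animated alpha-composite UI material for the Discovery social live activity visual, drawing a glowing circular rift.
# Wave-warped UVs, circle SDFs and exclusion blends form wobbly rings, alongside a twice-sampled rotating rift texture, tiled oscillating particles and a sweeping radial glow.
# Layers combine through Max and Add, with an animated cosine-gradient palette available for tint.
# Masks fade the bottom to clear Creator Tile titles, and vertex alpha scales the result.
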Material /Game/UI/DiscoveryShared/Art/Materials/M_UI_SocialLiveActivity
graph {
 properties {
  bUsedWithStaticMesh=false
  blendMode=BLEND_AlphaComposite
  materialDomain=MD_UI
  pixelDepthOffsetMode=PDOM_Legacy
  usedShadingModels=Default Lit
 }
 nodes [
  n1 = NamedRerouteUsage() {desc=Masking Bottom to avoid overlapping with Creator Tiles titles}
  n2 = NamedRerouteUsage()
  n3 = Saturate(None=Multiply(A=Min(A=@n1, B=@n2)) {constB=2})
  Multiply(A=@n3, B=@n3)
  n4 = VertexColor()
  n5 = NamedRerouteUsage()
  n6 = Multiply(A=@n5, B=NamedRerouteUsage() {desc=Adding Variety to the RIngs})
  n7 = Multiply(A=LinearInterpolate(A=@n6, B=@n5, Alpha=@n6)) {constB=2}
  n8 = MaterialFunctionCall(DistanceField=NamedRerouteUsage(), Threshold=Constant() {r=0.0005}) {materialFunction=&MF_UI_AntiAliasing}
  n9 = NamedRerouteUsage()
  n10 = NamedRerouteUsage() {desc=Adding variety to solid shape}
  n11 = Saturate(None=Multiply(A=Multiply(A=@n10, B=@n10)) {constB=2})
  n12 = NamedRerouteUsage() {desc=BG Rift}
  n13 = Max(
   A=Max(A=Saturate(None=@n7), B=Multiply(A=@n8.Inner) {constB=0.25, desc=Opacity})
   B=Max(A=Multiply(A=@n9, B=@n11), B=Multiply(A=@n12) {constB=0.2, desc=Opacity})
  )
  n14 = Multiply(A=NamedRerouteUsage()) {constB=2}
  n15 = Multiply(A=Saturate(None=Add(A=@n13, B=@n14)), B=@n3)
  Multiply(A=@n4.A, B=Multiply(A=NamedRerouteUsage(), B=@n15))
  Multiply(A=@n4.A, B=Multiply(A=@n15) {constB=0.5})
  n16 = Time()
  n17 = Multiply(A=@n16, B=ScalarParameter() {defaultValue=0.125, parameterName=TimeSpeed})
  n18 = Panner(
   Coordinate=MaterialFunctionCall(UV=Rotator(Time=@n16)) {materialFunction=&MF_UI_UVScaling}.Scaled to Fit
   Time=@n17
  ) {speedX=1}
  n19 = AppendVector(
   A=ScalarParameter() {defaultValue=24, group=Tiling, parameterName=TileX}
   B=ScalarParameter() {defaultValue=20, group=Tiling, parameterName=TileY}
  )
  n20 = Multiply(A=Frac(None=@n18), B=@n19)
  n21 = Multiply(A=Length(None=Ceil(None=@n20)) {is=class MaterialExpressionLength}, B=@n17)
  NamedRerouteDeclaration(None=Saturate(None=Cosine(None=@n21) {period=6.2832})) {name=OscillatingFracTiling, nodeColor=<rgba 0 0 0 1>}
  NamedRerouteDeclaration(None=MaterialFunctionCall() {materialFunction=&GetUserInterfaceUV}.Normalized UV) {name=UV, nodeColor=<rgba 0 0 0 1>}
  NamedRerouteDeclaration(None=MaterialFunctionCall!(DistanceFieldA=NamedRerouteUsage(), DistanceFieldB=NamedRerouteUsage(), BlendAmount) {materialFunction=/FortUILibrary/MaterialFunctions/Shapes/Modifiers/MF_UI_SDF_Exclusion}.Exact) {name=2 Rings Exclude SDF, nodeColor=<rgba 0 0 0 1>}
  n22 = ComponentMask(None=NamedRerouteUsage()) {g=true}
  NamedRerouteDeclaration(None=MaterialFunctionCall("Value (0-1)"=OneMinus(None=@n22)) {materialFunction=&MF_UI_EaseCurves}.EaseInCubic) {name=LinearGradient Mask, nodeColor=<rgba 0 0 0 1>}
  n23 = MaterialFunctionCall(UVs=NamedRerouteUsage(), Size=Constant() {desc=Scale, r=0.71}, OutlineThickness=Constant() {desc=Thickness, r=0.0008}) {materialFunction=&MF_UI_Circle}
  n24 = Abs(None=@n23.SDF)
  n25 = Constant()
  n26 = Constant() {desc=Thickness, r=0.008}
  NamedRerouteDeclaration(None=OneMinus(None=MaterialFunctionCall(Input=@n24, Input Low=@n25, Input High=@n26) {materialFunction=&RemapValue_0-1}.Result)) {name=Ring2, nodeColor=<rgba 0 0 0 1>}
  n27 = MaterialFunctionCall(
   UVs=NamedRerouteUsage() {desc=UVs with different wraps}
   Size=Constant() {desc=Scale, r=0.73}
   OutlineThickness=Constant() {desc=Thickness, r=0.001}
  ) {materialFunction=&MF_UI_Circle}
  n28 = Abs(None=@n27.SDF)
  n29 = Constant()
  n30 = Constant() {desc=Thickness, r=0.01}
  NamedRerouteDeclaration(None=OneMinus(None=MaterialFunctionCall(Input=@n28, Input Low=@n29, Input High=@n30) {materialFunction=&RemapValue_0-1}.Result)) {name=Ring1, nodeColor=<rgba 0 0 0 1>}
  NamedRerouteDeclaration(None=@n27.SDF) {name=Circle1_SDF, nodeColor=<rgba 0 0 0 1>}
  NamedRerouteDeclaration(None=@n23.SDF) {name=Circle2_SDF, nodeColor=<rgba 0 0 0 1>}
  n31 = MaterialFunctionCall(UVs=NamedRerouteUsage(), Size=Constant() {r=0.14}) {materialFunction=&MF_UI_Circle}
  n32 = Constant() {r=-1}
  n33 = Constant() {r=0.85}
  n34 = OneMinus(None=Saturate(None=MaterialFunctionCall(Input=@n31.SDF, Input Low=@n32, Input High=@n33) {materialFunction=&RemapValueRange}.Result))
  NamedRerouteDeclaration(None=MaterialFunctionCall("Value (0-1)"=MaterialFunctionCall("Value (0-1)"=@n34) {materialFunction=&MF_UI_EaseCurves}.EaseOutCubic) {materialFunction=&MF_UI_EaseCurves}.EaseOutCubic) {name=RadialGadient Mask, nodeColor=<rgba 0 0 0 1>}
  NamedRerouteDeclaration(None=@n28) {name=Ring1_SDF, nodeColor=<rgba 0 0 0 1>}
  n35 = ScalarParameter() {defaultValue=0.25, parameterName=XOffset_Start}
  n36 = ScalarParameter() {defaultValue=0.45, parameterName=XOffset_Stop}
  n37 = NamedRerouteUsage()
  n38 = ScalarParameter() {defaultValue=0.25, parameterName=YOffset_Start}
  n39 = ScalarParameter() {defaultValue=0.75, parameterName=YOffset_Stop}
  n40 = NamedRerouteUsage()
  n41 = MaterialFunctionCall(
   UVs=Frac(None=Multiply(A=@n18, B=@n19) {constB=16})
   CenterPosition=AppendVector(
    A=LinearInterpolate(A=@n35, B=@n36, Alpha=@n37) {constA=0.5, constB=0.05}
    B=LinearInterpolate(A=@n38, B=@n39, Alpha=@n40) {constA=0.5, constB=0.05}
   )
   Radius=LinearInterpolate(B=ScalarParameter() {defaultValue=0.125, parameterName=Radius}, Alpha=NamedRerouteUsage()) {constB=0.125}
   Density=LinearInterpolate(B=ScalarParameter() {defaultValue=1.25, parameterName=Density}, Alpha=NamedRerouteUsage()) {constB=0.125}
  ) {materialFunction=&RadialGradientExponential}
  n42 = NamedRerouteUsage()
  NamedRerouteDeclaration(None=Saturate(None=DotProduct(A=@n41.RadialGradientExponential, B=@n42))) {name=RiftParticles, nodeColor=<rgba 0 0 0 1>}
  NamedRerouteDeclaration(None=@n24) {name=Ring2_SDF, nodeColor=<rgba 0 0 0 1>}
  n43 = NamedRerouteUsage()
  n44 = Constant() {desc=Scale, r=0.77}
  NamedRerouteDeclaration(None=Abs(None=MaterialFunctionCall(UVs=@n43, Size=@n44) {materialFunction=&MF_UI_Circle}.SDF)) {name=Ring3_SDF, nodeColor=<rgba 0 0 0 1>}
  n45 = MaterialFunctionCall(DistanceFieldA=NamedRerouteUsage(), DistanceFieldB=NamedRerouteUsage()) {materialFunction=&MF_UI_SDF_Exclusion}
  n46 = NamedRerouteUsage()
  NamedRerouteDeclaration(None=MaterialFunctionCall(
    DistanceField=MaterialFunctionCall(DistanceFieldA=@n45.Exact, DistanceFieldB=@n46) {materialFunction=&MF_UI_SDF_Exclusion}.Exact
    Threshold=Constant() {r=0.01}
   ) {materialFunction=&MF_UI_AntiAliasing}.Inner) {name=RingsUnion, nodeColor=<rgba 0 0 0 1>}
  n47 = MaterialFunctionCall(UVs=NamedRerouteUsage(), Texture Scale=Constant2Vector() {g=0.9, r=0.9}) {materialFunction=&ScaleUVsByCenter}
  n48 = Multiply(A=Time(), B=Constant() {r=1.25})
  n49 = TextureObject() {samplerType=SAMPLERTYPE_LinearColor, texture=/Game/UI/DiscoveryShared/Art/T_UI_CircleRift}
  n50 = NamedRerouteDeclaration(None=TextureSample(UVs=Rotator(Coordinate=@n47.UVs, Time=@n48), Tex=@n49) {samplerType=SAMPLERTYPE_LinearColor, texture=&T_UI_CircleRift}.R) {name=RiftCircle_R, nodeColor=<rgba 0 0 0 1>}
  n51 = MaterialFunctionCall(UVs=NamedRerouteUsage(), Texture Scale=Constant2Vector() {g=1.05, r=1.05}) {materialFunction=&ScaleUVsByCenter}
  n52 = Multiply(A=Time(), B=Constant() {r=2})
  n53 = NamedRerouteDeclaration(None=TextureSample(UVs=Rotator(Coordinate=@n51.UVs, Time=@n52), Tex=@n49) {samplerType=SAMPLERTYPE_LinearColor, texture=&T_UI_CircleRift}.B) {name=RiftCircle_B, nodeColor=<rgba 0 0 0 1>}
  NamedRerouteDeclaration(None=Saturate(None=Add(A=@n50, B=@n53))) {name=Rotating Rift, nodeColor=<rgba 0 0 0 1>}
  n54 = MaterialFunctionCall(R=Multiply(A=Time()) {constB=0.2}, G=Constant() {r=0.33}, B=Constant() {r=0.67}) {materialFunction=&AppendMany}
  NamedRerouteDeclaration(None=LinearInterpolate(
    A=Constant4Vector() {constant=<rgba 0 1 0.5146 1>}.RGBA
    B=Constant4Vector() {constant=<rgba 0 0.283 0.3385 1>}.RGBA
    Alpha=MaterialFunctionCall!(Gradient, a, b, c, d=@n54.RGB) {materialFunction=/FortUILibrary/MaterialFunctions/Color/MF_UI_CosineGradient}.Result
   ) {constAlpha=0, constB=0}) {name=Kevin's Rainbow, nodeColor=<rgba 0 0 0 1>}
  NamedRerouteDeclaration(None=Max(A=NamedRerouteUsage(), B=NamedRerouteUsage())) {name=2 Rings, nodeColor=<rgba 0 0 0 1>}
  n55 = ScalarParameter() {defaultValue=2, parameterName=RotationSpeed}
  n56 = Time()
  n57 = MaterialFunctionCall!(Vector or UVs=Rotator(Time=Multiply(A=@n55, B=@n56) {constB=-1}) {speed=1}) {materialFunction=/FortUILibrary/MaterialFunctions/Basics/MF_UI_CheaperVectorToRadial}
  n58 = Multiply(A=@n57.Vector Converted to Angle) {constB=16}
  n59 = LinearInterpolate(B=Saturate(None=Multiply(A=@n58, B=@n58)), Alpha=SmoothStep(Value=@n57.Vector Converted to Angle) {constMax=0, constMin=0.95})
  n60 = OneMinus(None=@n57.Vector Converted to Angle)
  n61 = ScalarParameter() {defaultValue=-0.125, parameterName=GlowMin}
  n62 = ScalarParameter() {defaultValue=0.5, parameterName=GlowMax}
  n63 = Saturate(None=MaterialFunctionCall(
    UVs=NamedRerouteUsage()
    Size=ScalarParameter() {defaultValue=0.8, parameterName=GlowSize}
    GlowMin=DotProduct(A=@n60, B=@n61)
    GlowMax=DotProduct(A=@n57.Vector Converted to Angle, B=@n62)
   ) {materialFunction=&MF_UI_Circle}.Outline)
  NamedRerouteDeclaration(None=Saturate(None=DotProduct(A=@n59, B=@n63))) {name=RadialGlow, nodeColor=<rgba 0 0 0 1>}
  NamedRerouteDeclaration(None=MaterialFunctionCall!(
    UVs=NamedRerouteUsage()
    Amplitude X=Constant() {r=0.03}
    Amplitude Y=Constant() {r=0.004}
    Frequency X=Constant() {r=3}
    Frequency Y=Constant() {r=2}
    Phase=Constant() {r=1}
    Phase Speed=Constant() {r=0.5}
   ) {desc=+ X Amplitude and Y Frequency, materialFunction=/Game/UI/DiscoveryShared/Art/MaterialFunctions/MF_UI_MPT_WaveWarp}.Warped UVs) {name=UV_Wrap_A, nodeColor=<rgba 0 0 0 1>}
  NamedRerouteDeclaration(None=MaterialFunctionCall(
    UVs=NamedRerouteUsage()
    Amplitude X=Constant() {r=0.008}
    Amplitude Y=Constant() {r=0.008}
    Frequency X=Constant() {r=2.5}
    Frequency Y=Constant() {r=0.2}
    Phase=Constant() {r=1}
    Phase Speed=Constant() {r=0.8}
   ) {desc=Chill on Y + a bit faster, materialFunction=&MF_UI_MPT_WaveWarp}.Warped UVs) {name=UV_Wrap_B, nodeColor=<rgba 0 0 0 1>}
  NamedRerouteDeclaration(None=MaterialFunctionCall(
    UVs=NamedRerouteUsage()
    Amplitude X=Constant() {r=0.005}
    Amplitude Y=Constant() {r=0.005}
    Frequency X=Constant() {r=2}
    Frequency Y=Constant() {r=1}
    Phase=Constant() {r=1}
    Phase Speed=Constant() {r=0.5}
   ) {desc=Chill Wrap, materialFunction=&MF_UI_MPT_WaveWarp}.Warped UVs) {name=UV_Wrap_C, nodeColor=<rgba 0 0 0 1>}
 ]
}
